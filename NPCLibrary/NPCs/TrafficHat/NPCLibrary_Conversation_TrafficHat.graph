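# Conversation database for the TrafficHat NPC, a chemist character, defining the greeting and menu shown when a player talks to them.
# A single entry point leads to a speech node that picks a contextual line: first-ever meeting, NPC conversation-count checks, or higher-priority reactions to specific player outfits.
# The speech node then fans out through link nodes to shared conversations for quests, services, Season 34 story, healing expertise and TrafficHat's own dialogue.
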ConversationDatabase /NPCLibrary/NPCs/TrafficHat/NPCLibrary_Conversation_TrafficHat
entry_points [{tag=Conversation.Entry.TrafficHat, destinations [#1]}]
nodes [
 ConversationEntryPointNode_7 {
  id=#1
  class=ConversationEntryPointNode
  outputs [#2]
  properties {entryTag {tagName=Conversation.Entry.TrafficHat}}
 }
 Speech_C_4 {
  id=#2
  class=Speech_C
  outputs [
   #3
   #4
   #5
   #6
   #7
   #8
   #9
   #10
  ]
  properties {
   nodeName=""
   generalConfig {
    defaultMessage=Don't start any caustic reactions.
    contextualMessages [
     {
      message=I've got chemical concoctions for all of your robbing-and-stealing needs.
      contextRequirements [{requirement {class=FortControllerRequirement_IsFirstEverConversationWithNPC}}]
      priority {curve {rowName=Tandem.Speech.Priority.FirstEverInteraction}}
     }
     {
      message=Don't start any caustic reactions.
      contextRequirements [{requirement {class=FortControllerRequirement_CheckNPCMatchConversationCount, properties {nPCUniqueIdTag {tagName=AISpawnerData.Type.Tandem.TrafficHat}}}}]
      priority {curve {rowName=None}}
     }
     {
      message=What's brewing?
      contextRequirements [{requirement {class=FortControllerRequirement_CheckNPCMatchConversationCount, properties {nPCUniqueIdTag {tagName=AISpawnerData.Type.Tandem.TrafficHat}}}}]
      priority {curve {rowName=None}}
     }
     {
      message=It's not magic. It's chemistry.
      contextRequirements [{requirement {class=FortControllerRequirement_CheckNPCMatchConversationCount, properties {nPCUniqueIdTag {tagName=AISpawnerData.Type.Tandem.TrafficHat}}}}]
      priority {curve {rowName=None}}
     }
     {
      message=You are a scientific marvel.
      contextRequirements [{requirement {class=FortControllerRequirement_HasCID, properties {allowedCIds [/BRCosmetics/Athena/Items/Cosmetics/Characters/Character_WartyBrine_Sus]}}}]
      priority {curve {rowName=Tandem.Speech.Priority.OutfitReaction}}
     }
     {
      message=Nobody helps anybody for free around here. What's your angle?
      contextRequirements [{requirement {class=FortControllerRequirement_HasCID, properties {allowedCIds [&Character_AuraCop_Bullet]}}}]
      priority {curve {rowName=Tandem.Speech.Priority.OutfitReaction}}
     }
     {
      message=Any luck on fixing that heating element for my sample incubator?
      contextRequirements [{requirement {class=FortControllerRequirement_HasCID, properties {allowedCIds [/BRCosmetics/Athena/Items/Cosmetics/Characters/Character_TealMink_Pelt]}}}]
      priority {curve {rowName=Tandem.Speech.Priority.OutfitReaction}}
     }
    ]
   }
  }
 }
 ConversationLinkNode_7 {id=#10, class=ConversationLinkNode, properties {remoteEntryTag {tagName=Conversation.Entry.DeluxeServices}}}
 ConversationLinkNode_4 {id=#9, class=ConversationLinkNode, properties {remoteEntryTag {tagName=Conversation.Entry.Expertise.Healing}}}
 ConversationLinkNode_5 {
  id=#8
  class=ConversationLinkNode
  properties {remoteEntryTag {tagName=Conversation.Entry.Season34.Expertise.Healing.Introduction}}
 }
 ConversationLinkNode_6 {
  id=#7
  class=ConversationLinkNode
  properties {remoteEntryTag {tagName=Conversation.Entry.Season34.Expertise.Healing.Finish}}
 }
 ConversationLinkNode_3 {
  id=#6
  class=ConversationLinkNode
  properties {remoteEntryTag {tagName=Conversation.Entry.TrafficHat.Dialogue}}
 }
 ConversationLinkNode_2 {
  id=#5
  class=ConversationLinkNode
  properties {remoteEntryTag {tagName=Conversation.Entry.Season34.Story.P02.Q02}}
 }
 ConversationLinkNode_0 {id=#4, class=ConversationLinkNode, properties {remoteEntryTag {tagName=Conversation.Entry.BasicServices}}}
 ConversationLinkNode_1 {id=#3, class=ConversationLinkNode, properties {remoteEntryTag {tagName=Conversation.Entry.NPCLibrary.Quests}}}
]
