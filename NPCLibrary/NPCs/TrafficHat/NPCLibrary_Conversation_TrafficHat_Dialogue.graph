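# Conversation database for the TrafficHat NPC's "Ask about..." dialogue, a science-minded character sharing opinions on chemistry, gold and magic.
# One entry point leads to an opening Speech node that branches into three choice-labeled topics, a Back option and a link to the separate BaronEsc dialogue entry.
# Each topic is a linear chain of four Speech lines, and all three converge on a shared ReturnToStart node labeled BACK.
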ConversationDatabase /NPCLibrary/NPCs/TrafficHat/NPCLibrary_Conversation_TrafficHat_Dialogue
entry_points [{tag=Conversation.Entry.TrafficHat.Dialogue, destinations [#1]}]
nodes [
 ConversationEntryPointNode_0 {
  id=#1
  class=ConversationEntryPointNode
  outputs [#2]
  properties {entryTag {tagName=Conversation.Entry.TrafficHat.Dialogue}}
 }
 Speech_C_0 {
  id=#2
  class=Speech_C
  outputs [
   #3
   #4
   #5
   #6
   #7
  ]
  properties {nodeName="", softTaskIcon=&T_UI_NPC_question, generalConfig {defaultMessage=I sense a question bubbling up...}}
  sub_nodes [ConversationChoiceNode_5 {class=ConversationChoiceNode, properties {defaultChoiceDisplayText=Ask about...}}]
 }
 ConversationLinkNode_0 {
  id=#7
  class=ConversationLinkNode
  properties {remoteEntryTag {tagName=Conversation.Entry.TrafficHat.Dialogue.BaronEsc}}
 }
 Speech_C_6 {
  id=#6
  class=Speech_C
  outputs [#8]
  properties {
   nodeName=""
   softTaskIcon=&T_UI_NPC_question
   generalConfig {defaultMessage=I think magic is just science we don't understand yet.}
  }
  sub_nodes [ConversationChoiceNode_15 {class=ConversationChoiceNode, properties {defaultChoiceDisplayText=Magic}}]
 }
 Speech_C_8 {
  id=#8
  class=Speech_C
  outputs [#9]
  properties {
   nodeName=""
   softTaskIcon=&T_UI_NPC_question
   generalConfig {defaultMessage="We used to think miasmas caused illness, and volcanoes were angry gods..."}
  }
 }
 Speech_C_9 {
  id=#9
  class=Speech_C
  outputs [#10]
  properties {
   nodeName=""
   softTaskIcon=&T_UI_NPC_question
   generalConfig {defaultMessage="...and that the tooth fairy was an actual fairy, not a tiny dentist with boundary issues."}
  }
 }
 Speech_C_10 {
  id=#10
  class=Speech_C
  outputs [#11]
  properties {
   nodeName=""
   softTaskIcon=&T_UI_NPC_question
   generalConfig {defaultMessage=Folks around here think that makes me a buzzkill. But isn't science so much more exciting than magic?}
  }
 }
 ReturnToStart_C_23 {
  id=#11
  class=ReturnToStart_C
  properties {softTaskIcon=&T_UI_NPC_Back}
  sub_nodes [ConversationChoiceNode_32 {class=ConversationChoiceNode, properties {defaultChoiceDisplayText=BACK}}]
 }
 Speech_C_2 {
  id=#5
  class=Speech_C
  outputs [#12]
  properties {
   nodeName=""
   softTaskIcon=&T_UI_NPC_question
   generalConfig {defaultMessage="Gold. Symbol: Au. Atomic number: 79. "}
  }
  sub_nodes [ConversationChoiceNode_1 {class=ConversationChoiceNode, properties {defaultChoiceDisplayText=Gold}}]
 }
 Speech_C_3 {
  id=#12
  class=Speech_C
  outputs [#13]
  properties {
   nodeName=""
   softTaskIcon=&T_UI_NPC_question
   generalConfig {defaultMessage=It's soft-ish. Inert-esque. A great conductor...}
  }
 }
 Speech_C_4 {
  id=#13
  class=Speech_C
  outputs [#14]
  properties {
   nodeName=""
   softTaskIcon=&T_UI_NPC_question
   generalConfig {defaultMessage=...and some people spend their whole lives chasing it.}
  }
 }
 Speech_C_5 {
  id=#14
  class=Speech_C
  outputs [#11]
  properties {
   nodeName=""
   softTaskIcon=&T_UI_NPC_question
   generalConfig {defaultMessage="It pays for my beakers and keeps the lights on, but there are more interesting elements out there to study."}
  }
 }
 Back_C_0 {
  id=#4
  class=Back_C
  sub_nodes [ConversationChoiceNode_7 {class=ConversationChoiceNode, properties {defaultChoiceDisplayText=BACK}}]
 }
 Speech_C_7 {
  id=#3
  class=Speech_C
  outputs [#15]
  properties {nodeName="", softTaskIcon=&T_UI_NPC_question, generalConfig {defaultMessage=Chemistry is EVERYTHING.}}
  sub_nodes [ConversationChoiceNode_0 {class=ConversationChoiceNode, properties {defaultChoiceDisplayText=Chemistry}}]
 }
 Speech_C_15 {
  id=#15
  class=Speech_C
  outputs [#16]
  properties {
   nodeName=""
   softTaskIcon=&T_UI_NPC_question
   generalConfig {defaultMessage=Chemistry lets us combine elements in endless permutations to make something new.}
  }
 }
 Speech_C_14 {
  id=#16
  class=Speech_C
  outputs [#17]
  properties {
   nodeName=""
   softTaskIcon=&T_UI_NPC_question
   generalConfig {defaultMessage="Everything on the Island, good or bad, exists because of chemistry."}
  }
 }
 Speech_C_1 {
  id=#17
  class=Speech_C
  outputs [#11]
  properties {
   nodeName=""
   softTaskIcon=&T_UI_NPC_question
   generalConfig {defaultMessage=Who wouldn't want to master something like that?}
  }
 }
]
